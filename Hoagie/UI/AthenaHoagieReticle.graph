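# HUD reticle widget for the Hoagie flying vehicle, combining an artificial horizon with boost, speed, fuel and altitude bracket meters.
# Tick feeds smoothed pitch and roll plus boost duration or cooldown into dynamic materials, and plays a tint animation at critical health.
# Parent-class events drive speed and fuel meters, hidden without a fuel system.
# Boost delegates recolor the meter, and async gameplay-tag waits toggle a vehicle-disabled warning.
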WidgetBlueprint /Hoagie/UI/AthenaHoagieReticle
animations [
 Critical {
  binding_count=1
  length_seconds=0.4803
  object=&self:Critical
  sequence {bindings [CommonBorder_0 {
     id=#1
     possessed_class=CommonBorder
     tracks [{
       type=MovieSceneColorTrack
       display_name=Content Color and Opacity
       sections [{
         type=MovieSceneColorSection
         end_frame=9
         start_frame=0
         channels [
          Color.R {type=MovieSceneScriptingFloatChannel, keys [[0,0.8229],[9.6057,0.5365]]}
          Color.G {type=MovieSceneScriptingFloatChannel, keys [[0,0.0312],[9.6057,0.0203]]}
          Color.B {type=MovieSceneScriptingFloatChannel, const [0,2]}
          Color.A {type=MovieSceneScriptingFloatChannel, keys [[0,0.9],[9.6057,0.5]]}
         ]
        }]
      }]
    }]}
 }
 Intro {
  binding_count=5
  length_seconds=1.1421
  object=&self:Intro
  sequence {bindings [
    CommonBorder_2 {
     id=#2
     possessed_class=CommonBorder
     tracks [{
       type=MovieSceneFloatVectorTrack
       display_name=Desired Size Scale
       sections [{
         type=MovieSceneFloatVectorSection
         channels [
          Vector.X {type=MovieSceneScriptingFloatChannel}
          Vector.Y {type=MovieSceneScriptingFloatChannel, keys [[14.9163,0.7],[18.511,1]]}
         ]
        }]
      }]
    }
    CommonBorder_1 {
     id=#3
     possessed_class=CommonBorder
     tracks [{
       type=MovieSceneFloatVectorTrack
       display_name=Desired Size Scale
       sections [{
         type=MovieSceneFloatVectorSection
         channels [
          Vector.X {type=MovieSceneScriptingFloatChannel}
          Vector.Y {type=MovieSceneScriptingFloatChannel, keys [[14.9163,0.7],[18.511,1]]}
         ]
        }]
      }]
    }
    Image_BoostMeter {
     id=#4
     possessed_class=Image
     tracks [
      {
       type=MovieScene2DTransformTrack
       sections [{
         type=MovieScene2DTransformSection
         channels [
          Translation.X {type=MovieSceneScriptingFloatChannel, keys [[4.9287,-50],[14.9163,0]]}
          Translation.Y {type=MovieSceneScriptingFloatChannel}
          Angle {type=MovieSceneScriptingFloatChannel}
          Scale.X {type=MovieSceneScriptingFloatChannel}
          Scale.Y {type=MovieSceneScriptingFloatChannel}
          Shear.X {type=MovieSceneScriptingFloatChannel}
          Shear.Y {type=MovieSceneScriptingFloatChannel}
         ]
        }]
      }
      {
       type=MovieSceneFloatTrack
       display_name=Render Opacity
       sections [{
         type=MovieSceneFloatSection
         channels [None {type=MovieSceneScriptingFloatChannel, keys [[4.9287,0],[5.6087,1],[6.42,0],[7.179,1],[14.9163,1]]}]
        }]
      }
     ]
    }
    Image_Horizon {
     id=#5
     possessed_class=Image
     tracks [
      {
       type=MovieSceneWidgetMaterialTrack
       sections [{
         type=MovieSceneParameterSection
         channels [
          OuterRadius {type=MovieSceneScriptingFloatChannel, keys [[0,0.04],[21.2713,0.5]]}
          OuterRingWidth {type=MovieSceneScriptingFloatChannel, keys [[0,0.2435],[5.8443,0.02],[21.2713,0.0035]]}
          MaskSize {type=MovieSceneScriptingFloatChannel, keys [[0,-0.2568],[21.2713,0.3452]]}
          OuterRingMaskR {type=MovieSceneScriptingFloatChannel, keys [[0,0.0372],[21.2713,1]]}
          OuterRingMaskG {type=MovieSceneScriptingFloatChannel, keys [[0,-0.0526],[4,0.8701],[21.2713,0.2601]]}
          OuterRingMaskB {type=MovieSceneScriptingFloatChannel, const [0,2]}
          OuterRingMaskA {type=MovieSceneScriptingFloatChannel, const [1,2]}
         ]
        }]
      }
      {
       type=MovieSceneFloatTrack
       display_name=Render Opacity
       sections [{
         type=MovieSceneFloatSection
         channels [None {
           type=MovieSceneScriptingFloatChannel
           keys [[0,1],[2.2503,0],[2.9303,1],[3.7417,0],[4.5007,1],[20.4573,1],[20.5913,0],[21.2713,1],[22.0827,0],[22.8417,1]]
          }]
        }]
      }
     ]
    }
    Image_AltMeter {
     id=#6
     possessed_class=Image
     tracks [
      {
       type=MovieScene2DTransformTrack
       sections [{
         type=MovieScene2DTransformSection
         channels [
          Translation.X {type=MovieSceneScriptingFloatChannel, keys [[4.9287,50],[14.9163,0]]}
          Translation.Y {type=MovieSceneScriptingFloatChannel}
          Angle {type=MovieSceneScriptingFloatChannel}
          Scale.X {type=MovieSceneScriptingFloatChannel}
          Scale.Y {type=MovieSceneScriptingFloatChannel}
          Shear.X {type=MovieSceneScriptingFloatChannel}
          Shear.Y {type=MovieSceneScriptingFloatChannel}
         ]
        }]
      }
      {
       type=MovieSceneFloatTrack
       display_name=Render Opacity
       sections [{
         type=MovieSceneFloatSection
         channels [None {type=MovieSceneScriptingFloatChannel, keys [[4.9287,0],[5.6087,1],[6.42,0],[7.179,1],[14.9163,1]]}]
        }]
      }
     ]
    }
   ]}
 }
]
graphs [
 EventGraph {nodes [
   n1 = HandleBoostStateChanged!() -> (OutputDelegate, then, State:EHoagieBoostState Enum=Unknown) {is=title K2Node_CustomEvent}
   > UpdateBoostState!(execute, self, State:EHoagieBoostState Enum=@n1.State) -> (then)
   n2 = Event Tick()
   > n3 = Is Valid(InputObject=Get OwningHoagieVehicle() -> (OwningHoagieVehicle:Fort Hoagie Vehicle Object Reference))
   SetRenderOpacity(execute=@n3.Is Not Valid, InOpacity=0.2)
   SetRenderOpacity(execute=@n3.Is Valid, InOpacity=1)
   > UpdatePitchRoll!(execute, self, dt:double=@n2.InDeltaTime) -> (then)
   > UpdateBoostValue()
   > UpdateCriticalHealth()
   n4 = Event Setup MID!() -> (OutputDelegate, then, bVehicleUsesFuelSystem:bool) {is=title K2Node_Event}
   > n5 = Sequence!(execute) -> (then_0, then_1, then_2:Exec, then_3:Exec)
   Set VehicleMaxAltitude(
    execute=@n5.then_1
    VehicleMaxAltitude:double=GetMaxAltitude!(self:Fort Hoagie Vehicle Object Reference=Get OwningHoagieVehicle()) -> (ReturnValue:float)
   ) -> (Output_Get:double)
   n6 = GetFirstLocalFortPlayerController!(bIncludeNonSpectators:bool=true, bIncludeSpectators:bool=false) -> (ReturnValue:Fort Player Controller Object Reference)
   Set MaxAltitudeWithoutFallDamage(
    execute=@n5.then_0
    MaxAltitudeWithoutFallDamage:float=GetMaxDistanceWithoutFallDamage!(self:Fort Pawn Object Reference=GetPlayerPawn!(self:Fort Player Controller Object Reference=@n6) -> (ReturnValue:Fort Player Pawn Object Reference)) -> (ReturnValue:float)
   ) -> (Output_Get:float)
   SetVisibility(
    execute=@n5.then_3
    self=[
     Get CommonBorder_FuelMeterContainer() -> (CommonBorder_FuelMeterContainer:Common Border Object Reference)
     Get CommonBorder_SpeedometerContainer() -> (CommonBorder_SpeedometerContainer:Common Border Object Reference)
    ]
    InVisibility=Select(Option 0=Collapsed, Option 1=SelfHitTestInvisible, Index=@n4.bVehicleUsesFuelSystem)
   )
   Bind Event to On Boost State Changed(execute=@n5.then_2, self:Fort Hoagie Vehicle Object Reference=Get OwningHoagieVehicle(), Delegate=@n1) {is=title K2Node_AddDelegate}
   n7 = Get OwningHoagieVehicle()
   > n8 = Branch(Condition=IsBoosting!(self:Fort Vehicle Interface Interface=@n7) -> (ReturnValue:bool))
   > UpdateBoostState(State=Started)
   n9 = Branch(
    execute=@n8.else
    Condition="float > float"(A=GetBoostCooldown!(self:Fort Hoagie Vehicle Object Reference=@n7) -> (ReturnValue:float), B=0) {is=title K2Node_CallFunction}
   )
   > UpdateBoostState(State=Finished)
   UpdateBoostState(execute=@n9.else, State=Ready)
   n10 = Event OnVehicleUpdated!() -> (OutputDelegate, then, NewVehicle:Fort Athena Vehicle Object Reference) {is=title K2Node_Event}
   n11 = Get DisabledTag() -> (DisabledTag:Gameplay Tag Structure)
   > n12 = WaitGameplayTagAddToActor!(execute:Exec, TargetActor:Actor Object Reference=@n10.NewVehicle, Tag:Gameplay Tag Structure=@n11, OnlyTriggerOnce:bool=false) -> (then:Exec, AsyncTaskProxy:Ability Async Wait Gameplay Tag Added Object Reference, Added:Exec) {is=title K2Node_AsyncAction}
   > n13 = WaitGameplayTagRemoveFromActor!(execute:Exec, TargetActor:Actor Object Reference=@n10.NewVehicle, Tag:Gameplay Tag Structure=@n11, OnlyTriggerOnce:bool=false) -> (then:Exec, AsyncTaskProxy:Ability Async Wait Gameplay Tag Removed Object Reference, Removed:Exec) {is=title K2Node_AsyncAction}
   OnVehicleEnabled!(execute=@n13.Removed, self, Enabled:bool=true) -> (then)
   OnVehicleEnabled(execute=@n12.Added, Enabled=false)
   Event Construct()
   > UpdateBoostValue()
   > n14 = GetDynamicMaterial(self=Get Image_Horizon() -> (Image_Horizon:Image Object Reference))
   > Set HorizonMID(HorizonMID:Material Instance Dynamic Object Reference=@n14) -> (Output_Get:Material Instance Dynamic Object Reference)
   > n15 = GetDynamicMaterial(self=Get Image_AltMeter() -> (Image_AltMeter:Image Object Reference))
   > Set AltitudeMeterMID(AltitudeMeterMID:Material Instance Dynamic Object Reference=@n15) -> (Output_Get:Material Instance Dynamic Object Reference)
   n16 = Get Intro()
   > StopAnimation(InAnimation=@n16)
   > PlayAnimation(InAnimation=@n16)
   n17 = Event UpdateSpeedometer!() -> (OutputDelegate, then, NormalizedSpeed:float, RawSpeed:float) {is=title K2Node_Event}
   > n18 = GetDynamicMaterial(self=Get Image_Speedometer() -> (Image_Speedometer:Image Object Reference))
   > SetScalarParameterValue(self=@n18, ParameterName=BarHeight, Value=@n17.NormalizedSpeed)
   n19 = Event SetVehicleHasInfiniteFuel!() -> (OutputDelegate, then, bInfiniteFuel:bool) {is=title K2Node_Event}
   > Branch(Condition=@n19.bInfiniteFuel)
   > n20 = GetDynamicMaterial(self=Get Image_FuelMeter() -> (Image_FuelMeter:Image Object Reference))
   > SetScalarParameterValue(self=@n20, ParameterName=BarHeight, Value=1)
   > SetScalarParameterValue(self=@n20, ParameterName=CaretY, Value=1)
   > n21 = GetDynamicMaterial(self=Get Image_FuelMeter_1() -> (Image_FuelMeter_1:Image Object Reference))
   > SetScalarParameterValue(self=@n21, ParameterName=CaretY, Value=1)
   n22 = Event OnFuelValueChanged!() -> (OutputDelegate, then, FuelPercentage:float, bIsLowOnFuel:bool) {is=title K2Node_Event}
   > n23 = GetDynamicMaterial(self=Get Image_FuelMeter())
   > SetScalarParameterValue(self=@n23, ParameterName=BarHeight, Value=@n22.FuelPercentage)
   > SetScalarParameterValue(self=@n23, ParameterName=CaretY, Value=@n22.FuelPercentage)
   > n24 = GetDynamicMaterial(self=Get Image_FuelMeter_1())
   > SetScalarParameterValue(self=@n24, ParameterName=CaretY, Value=@n22.FuelPercentage)
  ]}
 OnVehicleEnabled {nodes [
   n1 = OnVehicleEnabled!() -> (then, Enabled:bool=~) {is=title K2Node_FunctionEntry}
   > n2 = Branch(Condition=@n1)
   n3 = Get WBP_Vehicle_Warning() -> (WBP_Vehicle_Warning:WBP Vehicle Warning Object Reference)
   > RemoveWarning!(execute, self:WBP Vehicle Warning Object Reference=@n3, Type:ENUM_WarningType Enum=NewEnumerator2) -> (then)
   > SetVisibility(self=@n3, InVisibility=Collapsed)
   ShowWarning!(execute=@n2.else, self:WBP Vehicle Warning Object Reference=@n3, Type:ENUM_WarningType Enum=NewEnumerator2) -> (then)
   > SetVisibility(self=@n3, InVisibility=HitTestInvisible)
  ]}
 UpdateBoostState {nodes [
   n1 = UpdateBoostState!() -> (then, State:EHoagieBoostState Enum=Unknown) {is=title K2Node_FunctionEntry}
   > n2 = GetDynamicMaterial(self=Get Image_BoostMeter() -> (Image_BoostMeter:Image Object Reference))
   > SetVectorParameterValue(
    self=@n2
    ParameterName=BarColor
    Value=Select!(
     Unknown:Linear Color Structure=(R=1,G=1,B=1,A=1)
     Ready:Linear Color Structure=(R=1,G=1,B=1,A=1)
     Started:Linear Color Structure=(R=1,G=0.5419,B=0,A=1)
     Finished:Linear Color Structure=(R=1,G=0.2279,B=0,A=0.839)
     Failed:Linear Color Structure=(R=1,G=1,B=1,A=1)
     Index:EHoagieBoostState Enum=@n1
    ) -> (ReturnValue:Linear Color Structure)
   )
   n3 = Switch on EHoagieBoostState!(execute, Selection:EHoagieBoostState Enum=@n1) -> (Unknown:Exec, Ready:Exec, Started:Exec, Finished:Exec, Failed:Exec) {is=title K2Node_SwitchEnum}
   SetFillColorAndOpacity!(
    execute=@n3.Finished
    self:Progress Bar Object Reference=Get ProgressBar_Boost() -> (ProgressBar_Boost:Progress Bar Object Reference)
    InColor:Linear Color Structure=LinearColor_Yellow!() -> (ReturnValue:Linear Color Structure)
   ) -> (then)
   SetFillColorAndOpacity(
    execute=[@n3.Failed, @n3.Ready, @n3.Unknown]
    self=Get ProgressBar_Boost()
    InColor=LinearColor_White!() -> (ReturnValue:Linear Color Structure)
   )
   SetFillColorAndOpacity(execute=@n3.Started, self=Get ProgressBar_Boost(), InColor=LinearColor_Red!() -> (ReturnValue:Linear Color Structure))
  ]}
 UpdateBoostValue {nodes [
   UpdateBoostValue() {is=title K2Node_FunctionEntry}
   n1 = Get OwningHoagieVehicle()
   > Is Valid(InputObject=@n1)
   > n2 = GetDynamicMaterial(self=Get Image_BoostMeter())
   n3 = Get OwningHoagieVehicle()
   n4 = GetMaxBoostCooldown!(self:Fort Hoagie Vehicle Object Reference=@n3) -> (ReturnValue:float)
   n5 = GetBoostCooldown(self=@n3)
   n6 = Get OwningHoagieVehicle()
   n7 = Select!(
    Option 0:double=float - float(B=float / float(A=@n5, B=@n4) {is=title K2Node_CallFunction}) {is=title K2Node_CallFunction}
    Option 1:double=float / float(
     A=GetBoostTimeLeft!(self:Fort Hoagie Vehicle Object Reference=@n6) -> (ReturnValue:float)
     B=GetBoostDuration!(self:Fort Hoagie Vehicle Object Reference=@n6) -> (ReturnValue:float)
    )
    Index:bool="float > float"(A=GetBoostTimeLeft(self=@n1), B=0)
   ) -> (ReturnValue:double)
   > SetScalarParameterValue(self=@n2, ParameterName=BarHeight, Value=@n7)
   > SetScalarParameterValue(self=@n2, ParameterName=CaretY, Value=@n7)
  ]}
 UpdateCriticalHealth {nodes [
   UpdateCriticalHealth() {is=title K2Node_FunctionEntry}
   > n1 = Branch(Condition=IsCriticalHealth!(self:Fort Athena Vehicle Object Reference=Get OwningHoagieVehicle()) -> (ReturnValue:bool))
   n2 = Get Critical() -> (Critical:Widget Animation Object Reference)
   > n3 = Branch(Condition=IsAnimationPlaying(InAnimation=@n2))
   PlayAnimation(execute=@n3.else, InAnimation=@n2)
   Branch(execute=@n1.else, Condition=IsAnimationPlaying(InAnimation=Get Critical()))
   > StopAnimation(InAnimation=@n2)
  ]}
 UpdatePitchRoll {nodes [
   n1 = Get text_pitchdebug() -> (text_pitchdebug:Common Text Object Reference)
   "SetText (Text)"(self=@n1, InText=Format Text!(Format="LOCTEXT(\"pitch {0}\\r\\nroll {1}\")", 0:Wildcard, 1:Wildcard) -> (Result))
   > SetVisibility(self=@n1, InVisibility=HitTestInvisible)
   n2 = UpdatePitchRoll!() -> (then, dt:double=~) {is=title K2Node_FunctionEntry}
   n3 = BreakRotator(InRot=Get Actor Rotation!(self:Actor Object Reference=Get OwningHoagieVehicle()) -> (ReturnValue:Rotator) {is=title K2Node_CallFunction})
   n4 = float / float(A=@n3.Pitch, B=-5)
   > n5 = Set CurrentPitch(CurrentPitch:double=FInterpTo(Current=Get CurrentPitch() -> (CurrentPitch:double), Target="Clamp (Float)"(Value=@n4, Min=-1, Max=1), DeltaTime=@n2, InterpSpeed=0.75)) -> (Output_Get:double)
   > SetScalarParameterValue(self=Get HorizonMID() -> (HorizonMID:Material Instance Dynamic Object Reference), ParameterName=Pitch, Value=@n5)
   n6 = float / float(A=@n3, B=-25)
   > n7 = Set CurrentRoll(CurrentRoll:double=FInterpTo(Current=Get CurrentRoll() -> (CurrentRoll:double), Target="Clamp (Float)"(Value=@n6, Min=-1, Max=1), DeltaTime=@n2, InterpSpeed=0.75)) -> (Output_Get:double)
   > SetScalarParameterValue(self=Get HorizonMID(), ParameterName=Rotation, Value=@n7)
  ]}
]
variables [
 CurrentPitch {type=number}
 CurrentRoll {type=number}
 DisabledTag {type=GameplayTag, default {tagName=Vehicle.Disabled}}
 HorizonMID {type=MaterialInstanceDynamic}
 VehicleMaxAltitude {type=number}
]
widgets {
 info {inheritedWidgetCount=10, parentClass=/Script/HoagieUI.FortHoagieVehicleReticle, rootWidgetClass=&Overlay, widgetCount=33}
 widgets [
  Overlay_8 {widgetClassPath=&Overlay, widget=Overlay_8}
  CommonBorder_0 {
   widgetClassPath=&CommonBorder
   parent=Overlay_8
   widget=CommonBorder_0
   properties {contentColorAndOpacity=<rgba 1 1 1 0.7>, padding=<ltrb 0 0 0 0>}
   slot {class=OverlaySlot, properties {horizontalAlignment=HAlign_Center, verticalAlignment=VAlign_Center}}
  }
  HorizontalBox_0 {
   widgetClassPath=&HorizontalBox
   parent=CommonBorder_0
   widget=HorizontalBox_0
   slot {class=BorderSlot, properties {padding=<ltrb 0 0 0 0>}}
  }
  CommonBorder_1 {
   widgetClassPath=&CommonBorder
   parent=HorizontalBox_0
   widget=CommonBorder_1
   properties {padding=<ltrb 0 0 0 0>}
   slot {class=HorizontalBoxSlot, properties {verticalAlignment=VAlign_Center}}
  }
  SizeBox_0 {widgetClassPath=&SizeBox, parent=CommonBorder_1, widget=SizeBox_0, slot {class=BorderSlot, properties {padding=<ltrb 0 0 0 0>}}}
  Image_BoostMeter {
   widgetClassPath=&Image
   bInherited=true
   bIsVariable=true
   parent=SizeBox_0
   widget=Image_BoostMeter
   properties {
    visibility=HitTestInvisible
    brush {
     bIsDynamicallyLoaded=false
     drawAs=Image
     imageType=NoImage
     mirroring=NoMirror
     resourceName=None
     resourceObject=/Game/Athena/HUD/Vehicles/Materials/M_UI_Vehicle_BracketMeter_Inst
     tiling=NoTile
     imageSize=<17.943 185.3434>
     margin=<ltrb 0 0 0 0>
     outlineSettings {
      bUseBrushTransparency=false
      roundingType=HalfHeightRadius
      width=0
      color {colorUseRule=UseColor_Specified, specifiedColor=<rgba 0 0 0 0>}
      cornerRadii=<0 0 0 0>
     }
     tintColor {colorUseRule=UseColor_Specified, specifiedColor=<rgba 1 1 1 1>}
     uVRegion {bIsValid=false, max=<0 0>, min=<0 0>}
    }
    renderTransform {angle=0, scale=<-1 1>, shear=<0 0>, translation=<0 0>}
   }
   slot {class=SizeBoxSlot}
  }
  CommonBorder_SpeedometerContainer {
   widgetClassPath=&CommonBorder
   bIsVariable=true
   parent=HorizontalBox_0
   widget=CommonBorder_SpeedometerContainer
   slot {class=HorizontalBoxSlot, properties {horizontalAlignment=HAlign_Right, verticalAlignment=VAlign_Center}}
  }
  SizeBox_2 {widgetClassPath=&SizeBox, parent=CommonBorder_SpeedometerContainer, widget=SizeBox_2, slot {class=BorderSlot}}
  Image_Speedometer {
   widgetClassPath=&Image
   bInherited=true
   bIsVariable=true
   parent=SizeBox_2
   widget=Image_Speedometer
   properties {
    visibility=HitTestInvisible
    brush {
     bIsDynamicallyLoaded=false
     drawAs=Image
     imageType=NoImage
     mirroring=NoMirror
     resourceName=None
     resourceObject=/Game/Athena/HUD/Vehicles/Materials/M_UI_Vehicle_BracketMeter_Left
     tiling=NoTile
     imageSize=<17.943 185.3434>
     margin=<ltrb 0 0 0 0>
     outlineSettings {
      bUseBrushTransparency=false
      roundingType=HalfHeightRadius
      width=0
      color {colorUseRule=UseColor_Specified, specifiedColor=<rgba 0 0 0 0>}
      cornerRadii=<0 0 0 0>
     }
     tintColor {colorUseRule=UseColor_Specified, specifiedColor=<rgba 1 1 1 1>}
     uVRegion {bIsValid=false, max=<0 0>, min=<0 0>}
    }
   }
   slot {class=SizeBoxSlot}
  }
  Overlay_6 {
   widgetClassPath=&Overlay
   parent=HorizontalBox_0
   widget=Overlay_6
   slot {class=HorizontalBoxSlot, properties {padding=<ltrb 8 0 8 0>}}
  }
  Image_Horizon {
   widgetClassPath=&Image
   bIsVariable=true
   parent=Overlay_6
   widget=Image_Horizon
   properties {
    visibility=HitTestInvisible
    brush {
     bIsDynamicallyLoaded=false
     drawAs=Box
     imageType=NoImage
     mirroring=NoMirror
     resourceName=None
     resourceObject=/Game/Athena/HUD/Vehicles/Materials/M_UI_VehicleHorizon_Inst
     tiling=NoTile
     imageSize=<500 500>
     margin=<ltrb 0 0 0 0>
     outlineSettings {
      bUseBrushTransparency=false
      roundingType=HalfHeightRadius
      width=0
      color {colorUseRule=UseColor_Specified, specifiedColor=<rgba 0 0 0 0>}
      cornerRadii=<0 0 0 0>
     }
     tintColor {colorUseRule=UseColor_Specified, specifiedColor=<rgba 1 1 1 1>}
     uVRegion {bIsValid=false, max=<0 0>, min=<0 0>}
    }
   }
   slot {class=OverlaySlot, properties {horizontalAlignment=HAlign_Center, verticalAlignment=VAlign_Center, padding=<ltrb 250 0 250 0>}}
  }
  Image_Caret {
   widgetClassPath=&Image
   parent=Overlay_6
   widget=Image_Caret
   properties {
    visibility=Collapsed
    brush {
     bIsDynamicallyLoaded=false
     drawAs=Image
     imageType=NoImage
     mirroring=NoMirror
     resourceName=None
     resourceObject=/Game/Athena/HUD/Vehicles/Materials/M_UI_VehicleReticleCaret_Inst
     tiling=NoTile
     imageSize=<64 64>
     margin=<ltrb 0 0 0 0>
     outlineSettings {
      bUseBrushTransparency=false
      roundingType=HalfHeightRadius
      width=0
      color {colorUseRule=UseColor_Specified, specifiedColor=<rgba 0 0 0 0>}
      cornerRadii=<0 0 0 0>
     }
     tintColor {colorUseRule=UseColor_Specified, specifiedColor=<rgba 0 1 0.1579 1>}
     uVRegion {bIsValid=false, max=<0 0>, min=<0 0>}
    }
    renderTransform {angle=0, scale=<1 1>, shear=<0 0>, translation=<0 10>}
   }
   slot {class=OverlaySlot, properties {horizontalAlignment=HAlign_Center, verticalAlignment=VAlign_Center}}
  }
  CommonBorder_FuelMeterContainer {
   widgetClassPath=&CommonBorder
   bIsVariable=true
   parent=HorizontalBox_0
   widget=CommonBorder_FuelMeterContainer
   properties {padding=<ltrb 0 0 0 0>}
   slot {class=HorizontalBoxSlot, properties {verticalAlignment=VAlign_Center}}
  }
  Overlay_0 {
   widgetClassPath=&Overlay
   parent=CommonBorder_FuelMeterContainer
   widget=Overlay_0
   slot {class=BorderSlot, properties {padding=<ltrb 0 0 0 0>}}
  }
  VerticalBox_48 {
   widgetClassPath=&VerticalBox
   parent=Overlay_0
   widget=VerticalBox_48
   properties {renderTransform {angle=0, scale=<1 1>, shear=<0 0>, translation=<5 0>}}
   slot {class=OverlaySlot, properties {verticalAlignment=VAlign_Fill, padding=<ltrb 0 -25 0 -25>}}
  }
  CommonTextBlock_79 {
   widgetClassPath=&CommonTextBlock
   parent=VerticalBox_48
   widget=CommonTextBlock_79
   properties {
    style=/FortUILibrary/Styles/Typography/Foundation/TextStyle-Base-XXS-B.TextStyle-Base-XXS-B_C
    text=F
    font {
     bForceMonospaced=false
     bMaterialIsStencil=false
     fontMaterial=null
     fontObject=&NotoSans
     letterSpacing=0
     monospacedWidth=1
     size=14
     skewAmount=0
     typefaceFontName=NotoSans-Bold
     outlineSettings {
      bApplyOutlineToDropShadows=false
      bMiteredCorners=false
      bSeparateFillAlpha=false
      outlineMaterial=null
      outlineSize=0
      outlineColor=<rgba 0 0 0 1>
     }
    }
    shadowOffset=<0 0>
   }
   slot {class=VerticalBoxSlot}
  }
  Spacer_145 {
   widgetClassPath=&Spacer
   parent=VerticalBox_48
   widget=Spacer_145
   slot {class=VerticalBoxSlot, properties {size {sizeRule=Fill, value=1}}}
  }
  CommonTextBlock_205 {
   widgetClassPath=&CommonTextBlock
   parent=VerticalBox_48
   widget=CommonTextBlock_205
   properties {
    text=E
    colorAndOpacity {colorUseRule=UseColor_Specified, specifiedColor=<rgba 1 0.117 0.0006 1>}
    font {
     bForceMonospaced=false
     bMaterialIsStencil=false
     fontMaterial=null
     fontObject=&NotoSans
     letterSpacing=0
     monospacedWidth=1
     size=14
     skewAmount=0
     typefaceFontName=NotoSans-Bold
     outlineSettings {
      bApplyOutlineToDropShadows=false
      bMiteredCorners=false
      bSeparateFillAlpha=false
      outlineMaterial=null
      outlineSize=0
      outlineColor=<rgba 0 0 0 1>
     }
    }
    shadowOffset=<0 0>
   }
   slot {class=VerticalBoxSlot}
  }
  Image_FuelMeter {
   widgetClassPath=&Image
   bIsVariable=true
   parent=Overlay_0
   widget=Image_FuelMeter
   properties {
    visibility=HitTestInvisible
    brush {
     bIsDynamicallyLoaded=false
     drawAs=Image
     imageType=NoImage
     mirroring=NoMirror
     resourceName=None
     resourceObject=/Game/Athena/HUD/Vehicles/Materials/M_UI_Vehicle_BracketMeter_Right
     tiling=NoTile
     imageSize=<17.943 185.3434>
     margin=<ltrb 0 0 0 0>
     outlineSettings {
      bUseBrushTransparency=false
      roundingType=HalfHeightRadius
      width=0
      color {colorUseRule=UseColor_Specified, specifiedColor=<rgba 0 0 0 0>}
      cornerRadii=<0 0 0 0>
     }
     tintColor {colorUseRule=UseColor_Specified, specifiedColor=<rgba 1 1 1 1>}
     uVRegion {bIsValid=false, max=<0 0>, min=<0 0>}
    }
    renderTransform {angle=0, scale=<-1 1>, shear=<0 0>, translation=<0 0>}
   }
   slot {class=OverlaySlot}
  }
  Image_FuelMeter_1 {
   widgetClassPath=&Image
   bIsVariable=true
   parent=Overlay_0
   widget=Image_FuelMeter_1
   properties {
    visibility=HitTestInvisible
    brush {
     bIsDynamicallyLoaded=false
     drawAs=Image
     imageType=NoImage
     mirroring=NoMirror
     resourceName=None
     resourceObject=/Game/Athena/HUD/Vehicles/Materials/M_UI_Vehicle_BracketMeter_FuelTank
     tiling=NoTile
     imageSize=<17.943 185.3434>
     margin=<ltrb 0 0 0 0>
     outlineSettings {
      bUseBrushTransparency=false
      roundingType=HalfHeightRadius
      width=0
      color {colorUseRule=UseColor_Specified, specifiedColor=<rgba 0 0 0 0>}
      cornerRadii=<0 0 0 0>
     }
     tintColor {colorUseRule=UseColor_Specified, specifiedColor=<rgba 1 1 1 1>}
     uVRegion {bIsValid=false, max=<0 0>, min=<0 0>}
    }
    renderTransform {angle=0, scale=<-1 1>, shear=<0 0>, translation=<0 0>}
   }
   slot {class=OverlaySlot, properties {padding=<ltrb -17 0 0 0>}}
  }
  CommonBorder_2 {
   widgetClassPath=&CommonBorder
   parent=HorizontalBox_0
   widget=CommonBorder_2
   slot {class=HorizontalBoxSlot, properties {horizontalAlignment=HAlign_Right, verticalAlignment=VAlign_Center}}
  }
  SizeBox_1 {widgetClassPath=&SizeBox, parent=CommonBorder_2, widget=SizeBox_1, slot {class=BorderSlot}}
  Image_AltMeter {
   widgetClassPath=&Image
   bIsVariable=true
   parent=SizeBox_1
   widget=Image_AltMeter
   properties {
    visibility=HitTestInvisible
    brush {
     bIsDynamicallyLoaded=false
     drawAs=Image
     imageType=NoImage
     mirroring=NoMirror
     resourceName=None
     resourceObject=&M_UI_Vehicle_BracketMeter_Inst
     tiling=NoTile
     imageSize=<17.943 185.3434>
     margin=<ltrb 0 0 0 0>
     outlineSettings {
      bUseBrushTransparency=false
      roundingType=HalfHeightRadius
      width=0
      color {colorUseRule=UseColor_Specified, specifiedColor=<rgba 0 0 0 0>}
      cornerRadii=<0 0 0 0>
     }
     tintColor {colorUseRule=UseColor_Specified, specifiedColor=<rgba 1 1 1 1>}
     uVRegion {bIsValid=false, max=<0 0>, min=<0 0>}
    }
   }
   slot {class=SizeBoxSlot}
  }
  text_pitchdebug {
   widgetClassPath=&CommonTextBlock
   bIsVariable=true
   parent=Overlay_8
   widget=text_pitchdebug
   properties {
    style=&TextStyle-Base_C
    text=Text Block
    visibility=Collapsed
    font {
     bForceMonospaced=false
     bMaterialIsStencil=false
     fontMaterial=null
     fontObject=&NotoSans
     letterSpacing=0
     monospacedWidth=1
     size=24
     skewAmount=0
     typefaceFontName=NotoSans-Regular
     outlineSettings {
      bApplyOutlineToDropShadows=false
      bMiteredCorners=false
      bSeparateFillAlpha=false
      outlineMaterial=null
      outlineSize=0
      outlineColor=<rgba 0 0 0 1>
     }
    }
    shadowOffset=<0 0>
   }
   slot {class=OverlaySlot, properties {horizontalAlignment=HAlign_Right}}
  }
  WBP_Vehicle_Warning {
   widgetClassPath=/Game/Athena/HUD/Vehicles/Dashboard/WBP_Vehicle_Warning
   bIsVariable=true
   parent=Overlay_8
   widget=WBP_Vehicle_Warning
   slot {class=OverlaySlot, properties {horizontalAlignment=HAlign_Center, verticalAlignment=VAlign_Center}}
  }
  Widget_FuelGauge {widgetClassPath=/Script/FortniteUI.FortVehicleFuelGauge, bInherited=true, widget=null}
  HorizontalBox_FuelCluster {widgetClassPath=&HorizontalBox, bInherited=true, widget=null}
  Image_BoostGlow {widgetClassPath=&Image, bInherited=true, widget=null}
  Overlay_BoostSparks {widgetClassPath=&Overlay, bInherited=true, widget=null}
  NumericText_Speed {widgetClassPath=/Script/CommonUI.CommonNumericTextBlock, bInherited=true, widget=null}
  Overlay_Gauge {widgetClassPath=&Overlay, bInherited=true, widget=null}
  Text_ItemName {widgetClassPath=&CommonTextBlock, bInherited=true, widget=null}
  TutorialHighlight {widgetClassPath=&TutorialHighlight}
 ]
}
